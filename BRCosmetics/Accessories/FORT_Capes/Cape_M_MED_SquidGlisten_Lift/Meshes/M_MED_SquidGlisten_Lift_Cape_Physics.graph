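# Physics asset for the M_MED SquidGlisten Lift cape, giving its dynamic bones and the wearer's body capsule shapes for simulation.
# Center, left and right chains of seven large capsules each hang from small MSTR bodies linked through backpack_root to spine_05.
# Chain joints allow mainly one swing axis, with tight twist and cross-swing limits.
# Torso, upper-arm, leg and foot capsules approximate the body, joined by free angular constraints.
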PhysicsAsset /BRCosmetics/Accessories/FORT_Capes/Cape_M_MED_SquidGlisten_Lift/Meshes/M_MED_SquidGlisten_Lift_Cape_Physics
bodies {
 backpack_root [{shapeType=Capsule, length=1, radius=1, shapeName=backpack_root_capsule}]
 calf_l [{shapeType=Capsule, length=24, radius=9, shapeName=calf_l_capsule, center=<-22 0 0>, rotation=<rot 0 -90 90>}]
 calf_r [{shapeType=Capsule, length=24, radius=9, shapeName=calf_r_capsule, center=<22 0 0>, rotation=<rot 0 -90 90>}]
 clavicle_l [{shapeType=Capsule, length=12, radius=9, shapeName=clavicle_l_capsule, center=<8 0 0>, rotation=<rot 90 0 0>}]
 clavicle_r [{shapeType=Capsule, length=12, radius=9, shapeName=clavicle_r_capsule, center=<-8 0 0>, rotation=<rot 90 0 0>}]
 dyn_cape_01 [{shapeType=Capsule, length=13, radius=15, shapeName=dyn_cape_01_capsule, center=<10.5889 0 16.5472>, rotation=<rot 90 0 0>}]
 dyn_cape_02 [{shapeType=Capsule, length=20, radius=15, shapeName=dyn_cape_02_capsule, center=<10.5889 0 16.5472>, rotation=<rot 90 0 0>}]
 dyn_cape_03 [{shapeType=Capsule, length=20, radius=15, shapeName=dyn_cape_03_capsule, center=<10.5889 0 16.5472>, rotation=<rot 90 0 0>}]
 dyn_cape_04 [{shapeType=Capsule, length=20, radius=15, shapeName=dyn_cape_04_capsule, center=<10.5889 0 16.5472>, rotation=<rot 90 0 0>}]
 dyn_cape_05 [{shapeType=Capsule, length=20, radius=15, shapeName=dyn_cape_05_capsule, center=<10.5889 0 16.5472>, rotation=<rot 90 0 0>}]
 dyn_cape_06 [{shapeType=Capsule, length=20, radius=15, shapeName=dyn_cape_06_capsule, center=<10.5889 0 16.5472>, rotation=<rot 90 0 0>}]
 dyn_cape_07 [{shapeType=Capsule, length=20, radius=15, shapeName=dyn_cape_07_capsule, center=<10.5889 0 16.5472>, rotation=<rot 90 0 0>}]
 dyn_cape_MSTR [{shapeType=Capsule, length=1, radius=1, shapeName=dyn_cape_MSTR_capsule}]
 dyn_cape_l_01 [{shapeType=Capsule, length=13, radius=15, shapeName=dyn_cape_l_01_capsule, center=<10.5889 0 16.5472>, rotation=<rot 90 0 0>}]
 dyn_cape_l_02 [{shapeType=Capsule, length=20, radius=15, shapeName=dyn_cape_l_02_capsule, center=<10.5889 0 16.5472>, rotation=<rot 90 0 0>}]
 dyn_cape_l_03 [{shapeType=Capsule, length=20, radius=15, shapeName=dyn_cape_l_03_capsule, center=<10.5889 0 16.5472>, rotation=<rot 90 0 0>}]
 dyn_cape_l_04 [{shapeType=Capsule, length=20, radius=15, shapeName=dyn_cape_l_04_capsule, center=<10.5889 0 16.5472>, rotation=<rot 90 0 0>}]
 dyn_cape_l_05 [{shapeType=Capsule, length=20, radius=15, shapeName=dyn_cape_l_05_capsule, center=<10.5889 0 16.5472>, rotation=<rot 90 0 0>}]
 dyn_cape_l_06 [{shapeType=Capsule, length=20, radius=15, shapeName=dyn_cape_l_06_capsule, center=<10.5889 0 16.5472>, rotation=<rot 90 0 0>}]
 dyn_cape_l_07 [{shapeType=Capsule, length=20, radius=15, shapeName=dyn_cape_l_07_capsule, center=<10.5889 0 16.5472>, rotation=<rot 90 0 0>}]
 dyn_cape_l_MSTR [{shapeType=Capsule, length=0.505, radius=0.505, shapeName=dyn_cape_l_MSTR_capsule, rotation=<rot 0 -90 0>}]
 dyn_cape_r_01 [{shapeType=Capsule, length=13, radius=15, shapeName=dyn_cape_r_01_capsule, center=<10.5889 0 16.5472>, rotation=<rot 90 0 0>}]
 dyn_cape_r_02 [{shapeType=Capsule, length=20, radius=15, shapeName=dyn_cape_r_02_capsule, center=<10.5889 0 16.5472>, rotation=<rot 90 0 0>}]
 dyn_cape_r_03 [{shapeType=Capsule, length=20, radius=15, shapeName=dyn_cape_r_03_capsule, center=<10.5889 0 16.5472>, rotation=<rot 90 0 0>}]
 dyn_cape_r_04 [{shapeType=Capsule, length=20, radius=15, shapeName=dyn_cape_r_04_capsule, center=<10.5889 0 16.5472>, rotation=<rot 90 0 0>}]
 dyn_cape_r_05 [{shapeType=Capsule, length=20, radius=15, shapeName=dyn_cape_r_05_capsule, center=<10.5889 0 16.5472>, rotation=<rot 90 0 0>}]
 dyn_cape_r_06 [{shapeType=Capsule, length=20, radius=15, shapeName=dyn_cape_r_06_capsule, center=<10.5889 0 16.5472>, rotation=<rot 90 0 0>}]
 dyn_cape_r_07 [{shapeType=Capsule, length=20, radius=15, shapeName=dyn_cape_r_07_capsule, center=<10.5889 0 16.5472>, rotation=<rot 90 0 0>}]
 dyn_cape_r_MSTR [{shapeType=Capsule, length=0.505, radius=0.505, shapeName=dyn_cape_r_MSTR_capsule, rotation=<rot 0 -90 0>}]
 foot_l [{shapeType=Capsule, length=12, radius=9, shapeName=foot_l_capsule, center=<-5 6 0>, rotation=<rot 90 90 0>}]
 foot_r [{shapeType=Capsule, length=12, radius=9, shapeName=foot_r_capsule, center=<5 -6 0>, rotation=<rot 90 90 0>}]
 pelvis [{shapeType=Capsule, length=6, radius=16, shapeName=pelvis_capsule, center=<0 -0.4952 0>}]
 spine_01 [{shapeType=Capsule, length=7.9438, radius=16, center=<-1.7965 1.9252 0>}]
 spine_02 [
  {shapeType=Capsule, length=5, radius=15, center=<-1.711 2.7331 0.0001>}
  {shapeType=Capsule, length=29.3956, radius=9, center=<2.5024 1.372 3.9825>, rotation=<rot 90 0 0>}
  {shapeType=Capsule, length=29.875, radius=9, center=<2.4704 1.3719 -3.9825>, rotation=<rot 90 0 0>}
 ]
 spine_03 [{shapeType=Capsule, length=5, radius=14, center=<0.1445 3.0018 0.0001>}]
 spine_04 [{shapeType=Capsule, length=5, radius=15, center=<2.766 3.3203 0.0002>}]
 spine_05 [{shapeType=Capsule, length=12, radius=14, center=<-2.2795 3.1371 0>}, {shapeType=Capsule, length=13.9518, radius=13, center=<0 5.056 0>}]
 thigh_l [{shapeType=Capsule, length=28, radius=9, shapeName=thigh_l_capsule, center=<-21 0 0>, rotation=<rot 90 0 0>}]
 thigh_r [{shapeType=Capsule, length=28, radius=9, shapeName=thigh_r_capsule, center=<21 0 0>, rotation=<rot 90 0 0>}]
 upperarm_l [{shapeType=Capsule, length=23, radius=9, shapeName=upperarm_l_capsule, center=<14 0 0>, rotation=<rot 90 0 0>}]
 upperarm_r [{shapeType=Capsule, length=23, radius=9, shapeName=upperarm_r_capsule, center=<-14 0 0>, rotation=<rot 90 0 0>}]
}
constraints [
 {
  bone1Name=spine_01
  bone2Name=pelvis
  swing1LimitDegrees=45
  swing1Motion=Free
  swing2LimitDegrees=45
  swing2Motion=Free
  twistLimitDegrees=45
  twistMotion=Free
 }
 {
  bone1Name=spine_02
  bone2Name=spine_01
  swing1LimitDegrees=45
  swing1Motion=Free
  swing2LimitDegrees=45
  swing2Motion=Free
  twistLimitDegrees=45
  twistMotion=Free
 }
 {
  bone1Name=spine_03
  bone2Name=spine_02
  swing1LimitDegrees=45
  swing1Motion=Free
  swing2LimitDegrees=45
  swing2Motion=Free
  twistLimitDegrees=45
  twistMotion=Free
 }
 {
  bone1Name=spine_04
  bone2Name=spine_03
  swing1LimitDegrees=45
  swing1Motion=Free
  swing2LimitDegrees=45
  swing2Motion=Free
  twistLimitDegrees=45
  twistMotion=Free
 }
 {
  bone1Name=spine_05
  bone2Name=spine_04
  swing1LimitDegrees=45
  swing1Motion=Free
  swing2LimitDegrees=45
  swing2Motion=Free
  twistLimitDegrees=45
  twistMotion=Free
 }
 {
  bone1Name=backpack_root
  bone2Name=spine_05
  swing1LimitDegrees=45
  swing1Motion=Free
  swing2LimitDegrees=45
  swing2Motion=Free
  twistLimitDegrees=45
  twistMotion=Free
 }
 {
  bone1Name=dyn_cape_MSTR
  bone2Name=backpack_root
  swing1LimitDegrees=45
  swing1Motion=Free
  swing2LimitDegrees=45
  swing2Motion=Free
  twistLimitDegrees=45
  twistMotion=Free
 }
 {bone1Name=dyn_cape_01, bone2Name=dyn_cape_MSTR, swing1LimitDegrees=5, swing2LimitDegrees=45, twistLimitDegrees=5}
 {bone1Name=dyn_cape_02, bone2Name=dyn_cape_01, swing1LimitDegrees=5, swing2LimitDegrees=45, twistLimitDegrees=5}
 {bone1Name=dyn_cape_03, bone2Name=dyn_cape_02, swing1LimitDegrees=5, swing2LimitDegrees=45, twistLimitDegrees=5}
 {bone1Name=dyn_cape_04, bone2Name=dyn_cape_03, swing1LimitDegrees=5, swing2LimitDegrees=45, twistLimitDegrees=5}
 {bone1Name=dyn_cape_05, bone2Name=dyn_cape_04, swing1LimitDegrees=5, swing2LimitDegrees=45, twistLimitDegrees=5}
 {bone1Name=dyn_cape_06, bone2Name=dyn_cape_05, swing1LimitDegrees=5, swing2LimitDegrees=45, twistLimitDegrees=5}
 {bone1Name=dyn_cape_07, bone2Name=dyn_cape_06, swing1LimitDegrees=5, swing2LimitDegrees=45, twistLimitDegrees=5}
 {
  bone1Name=dyn_cape_l_MSTR
  bone2Name=backpack_root
  swing1LimitDegrees=45
  swing1Motion=Free
  swing2LimitDegrees=45
  swing2Motion=Free
  twistLimitDegrees=45
  twistMotion=Free
 }
 {bone1Name=dyn_cape_l_01, bone2Name=dyn_cape_l_MSTR, swing1LimitDegrees=5, swing2LimitDegrees=45, twistLimitDegrees=5}
 {bone1Name=dyn_cape_l_02, bone2Name=dyn_cape_l_01, swing1LimitDegrees=5, swing2LimitDegrees=45, twistLimitDegrees=5}
 {bone1Name=dyn_cape_l_03, bone2Name=dyn_cape_l_02, swing1LimitDegrees=5, swing2LimitDegrees=45, twistLimitDegrees=5}
 {bone1Name=dyn_cape_l_04, bone2Name=dyn_cape_l_03, swing1LimitDegrees=5, swing2LimitDegrees=45, twistLimitDegrees=5}
 {bone1Name=dyn_cape_l_05, bone2Name=dyn_cape_l_04, swing1LimitDegrees=5, swing2LimitDegrees=45, twistLimitDegrees=5}
 {bone1Name=dyn_cape_l_06, bone2Name=dyn_cape_l_05, swing1LimitDegrees=5, swing2LimitDegrees=45, twistLimitDegrees=5}
 {bone1Name=dyn_cape_l_07, bone2Name=dyn_cape_l_06, swing1LimitDegrees=5, swing2LimitDegrees=45, twistLimitDegrees=5}
 {
  bone1Name=dyn_cape_r_MSTR
  bone2Name=backpack_root
  swing1LimitDegrees=45
  swing1Motion=Free
  swing2LimitDegrees=45
  swing2Motion=Free
  twistLimitDegrees=45
  twistMotion=Free
 }
 {bone1Name=dyn_cape_r_01, bone2Name=dyn_cape_r_MSTR, swing1LimitDegrees=5, swing2LimitDegrees=45, twistLimitDegrees=5}
 {bone1Name=dyn_cape_r_02, bone2Name=dyn_cape_r_01, swing1LimitDegrees=5, swing2LimitDegrees=45, twistLimitDegrees=5}
 {bone1Name=dyn_cape_r_03, bone2Name=dyn_cape_r_02, swing1LimitDegrees=5, swing2LimitDegrees=45, twistLimitDegrees=5}
 {bone1Name=dyn_cape_r_04, bone2Name=dyn_cape_r_03, swing1LimitDegrees=5, swing2LimitDegrees=45, twistLimitDegrees=5}
 {bone1Name=dyn_cape_r_05, bone2Name=dyn_cape_r_04, swing1LimitDegrees=5, swing2LimitDegrees=45, twistLimitDegrees=5}
 {bone1Name=dyn_cape_r_06, bone2Name=dyn_cape_r_05, swing1LimitDegrees=5, swing2LimitDegrees=45, twistLimitDegrees=5}
 {bone1Name=dyn_cape_r_07, bone2Name=dyn_cape_r_06, swing1LimitDegrees=5, swing2LimitDegrees=45, twistLimitDegrees=5}
 {
  bone1Name=clavicle_l
  bone2Name=spine_05
  swing1LimitDegrees=45
  swing1Motion=Free
  swing2LimitDegrees=45
  swing2Motion=Free
  twistLimitDegrees=45
  twistMotion=Free
 }
 {
  bone1Name=upperarm_l
  bone2Name=clavicle_l
  swing1LimitDegrees=45
  swing1Motion=Free
  swing2LimitDegrees=45
  swing2Motion=Free
  twistLimitDegrees=45
  twistMotion=Free
 }
 {
  bone1Name=upperarm_r
  bone2Name=clavicle_r
  swing1LimitDegrees=45
  swing1Motion=Free
  swing2LimitDegrees=45
  swing2Motion=Free
  twistLimitDegrees=45
  twistMotion=Free
 }
 {
  bone1Name=clavicle_r
  bone2Name=spine_05
  swing1LimitDegrees=45
  swing1Motion=Free
  swing2LimitDegrees=45
  swing2Motion=Free
  twistLimitDegrees=45
  twistMotion=Free
 }
 {
  bone1Name=thigh_l
  bone2Name=pelvis
  swing1LimitDegrees=45
  swing1Motion=Free
  swing2LimitDegrees=45
  swing2Motion=Free
  twistLimitDegrees=45
  twistMotion=Free
 }
 {
  bone1Name=calf_l
  bone2Name=thigh_l
  swing1LimitDegrees=45
  swing1Motion=Free
  swing2LimitDegrees=45
  swing2Motion=Free
  twistLimitDegrees=45
  twistMotion=Free
 }
 {
  bone1Name=thigh_r
  bone2Name=pelvis
  swing1LimitDegrees=45
  swing1Motion=Free
  swing2LimitDegrees=45
  swing2Motion=Free
  twistLimitDegrees=45
  twistMotion=Free
 }
 {
  bone1Name=calf_r
  bone2Name=thigh_r
  swing1LimitDegrees=45
  swing1Motion=Free
  swing2LimitDegrees=45
  swing2Motion=Free
  twistLimitDegrees=45
  twistMotion=Free
 }
 {
  bone1Name=foot_l
  bone2Name=calf_l
  swing1LimitDegrees=45
  swing1Motion=Free
  swing2LimitDegrees=45
  swing2Motion=Free
  twistLimitDegrees=45
  twistMotion=Free
 }
 {
  bone1Name=foot_r
  bone2Name=calf_r
  swing1LimitDegrees=45
  swing1Motion=Free
  swing2LimitDegrees=45
  swing2Motion=Free
  twistLimitDegrees=45
  twistMotion=Free
 }
]
